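# Unlit translucent mesh-particle material for the sparkle explosion effect of the Brite Bomber Deluxe glider.
# Emissive twinkles via a vertex-alpha-driven sine, colored by particle color or panning rainbow, with optional hue shift; a masked lens-flare texture supplies the shape.
# World position offset builds sprites from mesh geometry, with dynamic parameters driving size and expansion along normals.
# Medium and high quality non-mobile paths add scrolling vector-noise drift weighted by vertex blue.
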Material /BRCosmetics/Gadgets/Assets/VinderTech_GliderChute/Glider_BriteBomberDeluxe/FX/Materials/M_BriteBomberDeluxe_Sparkles_Explosion_Mesh
graph {
 properties {
  bEnableResponsiveAA=true
  bUsedWithMeshParticles=true
  blendMode=BLEND_Translucent
  opacityMaskClipValue=0.01
  pixelDepthOffsetMode=PDOM_Legacy
  shadingModel=MSM_Unlit
  twoSided=true
  usedShadingModels=Unlit
 }
 nodes [
  n1 = VertexColor()
  n2 = Sine(None=Multiply(A=@n1.A, B=Time()))
  n3 = ComponentMask(None=TextureCoordinate() {uTiling=0, vTiling=0}) {g=true}
  n4 = Multiply(
   A=TextureSample(UVs=Panner(Coordinate=@n3) {speedX=1.6667}) {texture=/Game/Global/Textures/Default/Masks_and_Gradients/T_Grad_Rainbow}.RGB
   B=ParticleColor().RGB
  )
  n5 = MaterialFunctionCall(LDR Input=Clamp(None=@n4), HDR Tint=@n4) {materialFunction=&DeriveHDRfromLDR}
  n6 = ParticleColor()
  n7 = Multiply(
   A=Clamp(None=Multiply(A=@n2) {constB=50}) {maxDefault=50, minDefault=0.25}
   B=ShadingPathSwitch(Default=StaticSwitchParameter(True=@n5.RemappedRange, False=@n6.RGB) {parameterName=Use Rainbow Color}, Mobile=@n6.RGB) {desc=Migrated from FeatureLevelSwitch}
  )
  n8 = Multiply(A=@n1.A) {constB=0.025}
  n9 = VertexColor()
  n10 = Multiply(A=AppendVector(A=AppendVector(A=@n9.R, B=@n9.R), B=@n9.B), B=@n7)
  MaterialFunctionCall(Emissive=StaticSwitchParameter(
    True=@n7
    False=MaterialFunctionCall(Hue Shift Percentage=@n8, Texture=@n10) {materialFunction=&HueShift}.Result
   ) {parameterName=Use Particle Color with NO Hue Shift}) {materialFunction=&MF_FX_EmissiveAdjustments}
  n11 = TextureSampleParameter2D() {parameterName=FlareTexture, texture=/Game/Effects/Fort_Effects/Textures/T_Hanover_LensFlare}
  n12 = VertexColor()
  Multiply(A=StaticComponentMaskParameter(None=AppendVector(A=@n11.RGB, B=@n11.RGB)) {defaultG=true, parameterName=FlareTexture_Mask}, B=@n12.A)
  n13 = DynamicParameter() {defaultValue=<rgba 100 1 1 1>, paramNames [SpriteSize, Expansion, Sprite Size Scale, Param4]}
  n14 = Multiply(A=AppendVector(A=@n13.SpriteSize, B=@n13.SpriteSize), B=@n13.Sprite Size Scale)
  n15 = Add(
   A=MaterialFunctionCall!(Center Location=WorldPosition().XYZ, 0-1 UVs, XY scale=Multiply(A=@n12.A, B=@n14), OverridePivot, "Normals (Optional)") {materialFunction=&Sprite}.World position offset
   B=Multiply(A=VertexNormalWS(), B=Multiply(A=@n12.A, B=@n13.Expansion))
  )
  n16 = ScalarParameter() {defaultValue=50, parameterName=Intensity}
  n17 = Subtract(A=WorldPosition().XYZ, B=ActorPositionWS())
  n18 = Multiply(A=ScalarParameter() {defaultValue=1, parameterName=Speed}, B=Time())
  n19 = VectorNoise(World Position=Divide(A=Subtract(A=@n17, B=@n18), B=ScalarParameter() {defaultValue=100, parameterName=Scale})) {noiseFunction=VNF_VectorALU}
  n20 = Add(A=@n15, B=Multiply(A=VertexColor().B, B=Multiply(A=@n16, B=@n19)))
  ShadingPathSwitch(Default=QualitySwitch(Default=@n15, Low=@n15, High=@n20, Medium=@n20), Mobile=@n15) {desc=Migrated from FeatureLevelSwitch}
 ]
}
